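# Lit material in the MoonLanding environment that layers dust over a tinted surface, flagged for instanced static meshes.
# Two MatLayerBlend_Standard calls add dust on upward-facing vertex normals, then fade from full dust at the actor's origin upward over the Height parameter.
# The surface layer has parameterized color, specular, roughness, subsurface and clear coat, plus emissive from inverted Fresnel, boosted where facing the camera.
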Material /Game/Environments/MoonLanding/MTemp
graph {
 properties {bUseMaterialAttributes=true, bUsedWithInstancedStaticMeshes=true, pixelDepthOffsetMode=PDOM_Legacy, usedShadingModels=Default Lit}
 nodes [
  n1 = ComponentMask(None=VertexNormalWS()) {b=true}
  n2 = WorldPosition()
  n3 = ActorPositionWS()
  n4 = ComponentMask(None=Divide(A=Subtract(A=@n2.XYZ, B=@n3), B=ScalarParameter() {defaultValue=85.7176, parameterName=Height})) {b=true}
  Add(A=@n1, B=@n4)
  n5 = MakeMaterialAttributes!(
   BaseColor=VectorParameter() {parameterName=Dust, defaultValue=<rgba 0.2812 0.2812 0.2812 0>}.RGB
   Metallic
   Specular=ScalarParameter() {defaultValue=0.2524, parameterName=DustS}
   Roughness=ScalarParameter() {defaultValue=1, parameterName=DustR}
   Anisotropy
   EmissiveColor
   Opacity
   OpacityMask
   Normal
   Tangent
   WorldPositionOffset
   SubsurfaceColor
   ClearCoat=Constant()
   ClearCoatRoughness
   AmbientOcclusion
   Refraction
   CustomizedUVs_0
   CustomizedUVs_1
   CustomizedUVs_2
   CustomizedUVs_3
   CustomizedUVs_4
   CustomizedUVs_5
   CustomizedUVs_6
   CustomizedUVs_7
   PixelDepthOffset
   ShadingModel
   Displacement
  )
  n6 = VectorParameter() {parameterName=C, defaultValue=<rgba 0.6146 0.5746 0.4545 0>}
  n7 = ScalarParameter() {defaultValue=0.2524, parameterName=S}
  n8 = ScalarParameter() {defaultValue=0.6857, parameterName=R}
  n9 = OneMinus(None=Fresnel() {baseReflectFraction=0, exponent=1.5})
  n10 = Saturate(None=Subtract(A=@n9, B=Constant() {r=0.7}))
  n11 = Constant() {r=5}
  n12 = Multiply(A=VectorParameter() {parameterName=E}.RGB, B=Add(A=@n9, B=Multiply(A=@n10, B=@n11)))
  n13 = Multiply(A=@n6.RGB, B=ScalarParameter() {parameterName=SSS})
  n14 = ScalarParameter() {parameterName=CC}
  n15 = ScalarParameter() {parameterName=CCR}
  n16 = Multiply(A=@n1, B=ScalarParameter() {defaultValue=0.5, parameterName=UpVec})
  MaterialFunctionCall!(
   Base Material=@n5
   Top Material=MaterialFunctionCall(
    Base Material=MakeMaterialAttributes(BaseColor=@n6.RGB, Specular=@n7, Roughness=@n8, EmissiveColor=@n12, SubsurfaceColor=@n13, ClearCoat=@n14, ClearCoatRoughness=@n15)
    Top Material=@n5
    Alpha=Saturate(None=@n16)
   ) {materialFunction=/Engine/Functions/MaterialLayerFunctions/MatLayerBlend_Standard}.Blended Material
   Alpha=Saturate(None=@n4)
  ) {materialFunction=&MatLayerBlend_Standard}
 ]
}
