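# Opaque world-space-normal material for Moon Landing space rocks, usable on Nanite, instanced and Niagara meshes.
# Height above MPC_MoonHeight's Ground_Z blends rock bases into world-XY-projected sand normals, masks and bottom color, darkening contact areas, with vertex-normal Z limiting some blends.
# World-aligned detail textures enrich rock normals and masks, while swirl noise varies diffuse brightness.
# A JustNormalTile static switch reduces the surface to tiled sand and flat color.
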Material /Game/Environments/MoonLanding/M_SpaceRocks
graph {
 properties {
  bTangentSpaceNormal=false
  bUsedWithInstancedStaticMeshes=true
  bUsedWithNanite=true
  bUsedWithNiagaraMeshParticles=true
  pixelDepthOffsetMode=PDOM_Legacy
  usedShadingModels=Default Lit
 }
 nodes [
  ComponentMask() {b=true}
  MaterialFunctionCall!(
   WSDeformationCenter=VectorParameter() {group=z_code, parameterName=WSDeformationCenter}.RGB
   Code_DestructionAmount=ScalarParameter() {group=z_code, parameterName=Code_DestructionAmount}
   Code_UseUvForDamage
   Code_UseHitEffects=StaticBoolParameter() {defaultValue=true, group=z_code, parameterName=Code_UseHitEffects}
   Code_Emissive
   WSDamageRadius=ScalarParameter() {defaultValue=1, group=z_code, parameterName=WSDamageRadius}
   CrackDamageShader
   Diffuse
   ScalarModDamageCracks
   ScalarOffsetforDamageCracks
   ExposeLerpFunctionForInvertingTheHitDistanceCalc
   InvertDamageFalloff
   CrackWorldPosition
   WorldPosition
   DistanceCullFadeAnimationOffset=ScalarParameter() {group=z_code, parameterName=DistanceCullFadeAnimationOffset}
  ) {materialFunction=/Game/Packages/jl_tests/Fortnite_HitEffects}
  ScalarParameter() {parameterName=WorldHeight}
  ScalarParameter() {group=z_code, parameterName=Code_HitAnimation}
  n1 = ComponentMask(None=Divide(A=WorldPosition().XYZ, B=ScalarParameter() {defaultValue=1024, parameterName=Scale})) {g=true, r=true}
  n2 = TextureSample(UVs=@n1) {samplerType=SAMPLERTYPE_Normal, texture=/Game/Environments/Sets/Arid/Textures/T_desert_sand_base_N}
  n3 = ScalarParameter() {defaultValue=814.5626, parameterName=125}
  n4 = MaterialFunctionCall(
   TextureObject=TextureObject() {samplerType=SAMPLERTYPE_Normal, texture=/Game/Environments/MoonLanding/Materials/Textures/SpaceDetails_N}
   TextureSize=@n3
   WorldSpace=StaticBool() {value=true}
  ) {materialFunction=&WorldAlignedNormal}
  n5 = ScalarParameter() {defaultValue=0.5, parameterName=DetailNormal}
  n6 = Add(
   A=Transform(None=TextureSampleParameter2D() {parameterName=Normal, samplerType=SAMPLERTYPE_Normal, texture=&FlatNormalBlue}.RGB)
   B=Multiply(A=Multiply(A=@n4.XYZ Texture, B=@n5), B=Constant3Vector() {constant=<rgba 1 1 0 1>})
  )
  n7 = ComponentMask(None=Subtract(A=WorldPosition().XYZ, B=CollectionParameter() {collection=/Game/Environments/MoonLanding/Meshes/BP/MPC_MoonHeight, parameterName=Ground_Z})) {b=true}
  n8 = Divide(A=@n7, B=ScalarParameter() {defaultValue=50, parameterName=BlurHeight})
  n9 = Saturate(None=@n8)
  n10 = MaterialFunctionCall(In=ComponentMask(None=VertexNormalWS()) {b=true}, Contrast=Constant() {r=2}) {materialFunction=&CheapContrast}
  StaticSwitchParameter(True=@n2.RGB, False=LinearInterpolate(A=@n6, B=LinearInterpolate(A=@n2.RGB, B=@n6, Alpha=@n9) {constB=0}, Alpha=@n10.Result)) {parameterName=JustNormalTile}
  n11 = VectorParameter() {parameterName=ColorBottom, defaultValue=<rgba 0.2812 0.2812 0.2812 0>}
  n12 = Constant() {r=0.75}
  n13 = Constant() {r=1.2}
  n14 = TextureSample(UVs=@n1) {samplerType=SAMPLERTYPE_LinearGrayscale, texture=/Game/Global/Textures/Default/Noise/T_SwirlNoise_H2}
  n15 = Saturate(None=@n8)
  n16 = LinearInterpolate(
   A=@n11.RGB
   B=Multiply(
    A=LinearInterpolate(A=@n12, B=@n13, Alpha=@n14.R)
    B=TextureSampleParameter2D() {parameterName=Diffuse, texture=/Game/Global/Textures/Default/Blanks/T_WhiteUserInterface}.RGB
   )
   Alpha=@n15
  )
  n17 = LinearInterpolate(A=Constant() {r=0.65}, Alpha=Saturate(None=Divide(A=@n7) {constB=10}))
  StaticSwitchParameter(True=@n11.RGB, False=LinearInterpolate(A=Multiply(A=@n16, B=@n17), B=@n16, Alpha=@n10.Result)) {parameterName=JustNormalTile}
  n18 = TextureSampleParameter2D() {parameterName=SpecularMasks, samplerType=SAMPLERTYPE_Masks, texture=/Game/Global/Textures/Default/Blanks/Temp_specular_Mask}
  n19 = TextureSample(UVs=@n1) {samplerType=SAMPLERTYPE_Masks, texture=/Game/Environments/Sets/Arid/Textures/T_desert_sand_base_S}
  StaticSwitchParameter(True=@n19.R, False=LinearInterpolate(A=@n19.R, B=@n18.R, Alpha=@n15)) {parameterName=JustNormalTile}
  n20 = MaterialFunctionCall(Float3=MaterialFunctionCall(
    TextureObject=TextureObject() {samplerType=SAMPLERTYPE_Masks, texture=/Game/Environments/MoonLanding/Materials/Textures/SpaceDetails_S}
    TextureSize=@n3
   ) {materialFunction=&WorldAlignedTexture}.XYZ Texture) {materialFunction=&BreakOutFloat3Components}
  StaticSwitchParameter(True=@n19.B, False=LinearInterpolate(A=@n19.B, B=Add(A=@n18.B, B=@n20.B), Alpha=@n15)) {parameterName=JustNormalTile}
  n21 = Constant()
  StaticSwitchParameter(True=@n21, False=LinearInterpolate(A=@n21, B=@n18.G, Alpha=@n15)) {parameterName=JustNormalTile}
  VectorParameter() {group=z_code, parameterName=Code_DeformationVector, defaultValue=<rgba 0 0 1 0>}
 ]
}
